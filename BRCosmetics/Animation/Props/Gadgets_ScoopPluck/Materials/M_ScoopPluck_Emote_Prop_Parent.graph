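# Parent material for the ScoopPluck emote prop, making masked emissive areas react to audio on top of the 2023 character shading.
# The character PreFX output passes through a car clear coat function, which also drives the clear coat normal, before reaching PostFX.
# Audio amplitude, shaped by exponent, scale and an animatable bias, sweeps a steep threshold across the FX texture's blue channel.
# Green modulates the resulting emissive and red limits where it replaces the base emissive.
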Material /BRCosmetics/Animation/Props/Gadgets_ScoopPluck/Materials/M_ScoopPluck_Emote_Prop_Parent
graph {
 properties {
  bUseMaterialAttributes=true
  bUsedWithClothing=true
  bUsedWithMeshDeformer=true
  bUsedWithMorphTargets=true
  bUsedWithSkeletalMesh=true
  pixelDepthOffsetMode=PDOM_Legacy
  shadingModel=MSM_FromMaterialExpression
  usedShadingModels=Default Lit | Subsurface | Clear Coat | Subsurface Profile | Cloth
 }
 nodes [
  n1 = MaterialFunctionCall() {materialFunction=&MF_CharacterParent_2023_PreFX}
  n2 = MaterialFunctionCall!(Mask=@n1.Tech Art Mask, Fresnel=@n1.Tech Art Mask, In=@n1.Tech Art Mask) {materialFunction=/Game/Global/MaterialFunctions/TechArt/MF_Advanced_CarClearCoat_v2}
  ClearCoatNormalCustomOutput!(None=@n2.ClearCoatNormal)
  n3 = GetMaterialAttributes(None=@n2.Result) {attributeGetTypes [#1]}
  n4 = TextureSample() {samplerType=SAMPLERTYPE_LinearColor, texture=/BRCosmetics/Animation/Props/Gadgets_ScoopPluck/Textures/T_ScoopPluck_Emote_Prop_FX2}
  n5 = MaterialFunctionCall() {materialFunction=/Game/Packages/Fortress_SharedMaterials/Base_Material_Functions/BaseCharacterShaderFunctions/Audio/MF_AudioReactivity}
  n6 = ScalarParameter() {defaultValue=2, parameterName=Amplitude_Exponent}
  n7 = Add(
   A=Multiply(A=Power(Base=@n5.Amplitude, Exp=@n6) {constExponent=1}, B=ScalarParameter() {defaultValue=1, parameterName=Amplitude_Scale})
   B=ScalarParameter() {desc=Should be animated, parameterName=Amplitude_Bias}
  )
  n8 = Multiply(A=Subtract(A=Saturate(None=@n7), B=@n4.B)) {constB=4}
  n9 = Multiply(A=@n3.Emissive Color, B=Multiply(A=@n4.G, B=Saturate(None=@n8)))
  MaterialFunctionCall(
   Material=SetMaterialAttributes(MaterialAttributes=@n3.MaterialAttributes, Emissive Color=LinearInterpolate(A=@n3.Emissive Color, B=@n9, Alpha=@n4.R)) {attributeSetTypes [#1]}
   M=@n1.M
   Pre Skinned Local Position=@n1.M
   Distance Blend=@n1.M
   Fresnel=@n1.M
  ) {materialFunction=&MF_CharacterParent_2023_PostFX}
 ]
}
